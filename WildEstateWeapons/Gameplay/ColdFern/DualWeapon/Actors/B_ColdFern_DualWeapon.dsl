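; ColdFern dual weapon actor that cycles muzzle flash, beam source and tracer origins through a list of muzzle sockets as it fires.
; Tracer and Niagara data events move the muzzle effect and pass FireIndex to shell ejection, then advance the wrapping index and update audio.
; A special-shot flag, cleared by a short timer, is exposed for animation blending and drives a fire sweetener sound played only for the locally viewed player.
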(event EventBeginPlay
  (|CallParentFunction)
  (EventDispatchers|BindEventtoOnBulletTracerBeamFXActivated (EventDispatchers|CreateEvent :!event Custom|OnBulletTrailFXActivated))
  (EventDispatchers|BindEventtoOnBulletTracerNiagaraDataSent (EventDispatchers|CreateEvent :!event Custom|BulletTracerNiagaraDataSent))
  (bind _asfort_player_pawn (Utilities|Casting|CastToFortPlayerPawn (Game|GetInstigator)))
  (Variables|Default|SetAs%20Fort%20Player%20Pawn _asfort_player_pawn))

(event EventEndPlay
  (params (EndPlayReason :type EEndPlayReason%20Enum :default "Destroyed"))
  (|CallParentFunction EndPlayReason)
  (EventDispatchers|BindEventtoOnBulletTracerBeamFXActivated (EventDispatchers|CreateEvent :!event Custom|OnBulletTrailFXActivated))
  (EventDispatchers|UnbindEventfromOnBulletTracerNiagaraDataSent (EventDispatchers|CreateEvent :!event Custom|BulletTracerNiagaraDataSent)))

(event FortWeapon|EventOnPlayImpactFX
  (params (HitResult :type Hit%20Result%20Structure%20%28by%20ref%29) (ImpactPhysicalSurface :type EPhysicalSurface%20Enum :default "SurfaceType_Default") (SpawnedPSC :type FXSystem%20Component%20Object%20Reference))
  (|CallParentFunction HitResult ImpactPhysicalSurface SpawnedPSC)
  (Effects|Components|ParticleSystem|SetFloatParameter SpawnedPSC "DebrisAmount" 0.4))

(event Custom|OnBulletTrailFXActivated
  (params (HitLocations :type Array%20of%20Vectors))
  (bind _spawned_muzzle_system (Variables|Effects|Muzzle|GetSpawned%20Muzzle%20System))
  ;[1 1 1] Muzzle Cycling Logic
  ;;[0 0 0 0.5] Updated muzzle FX after activating bullet trails.
  ;;;[1 1 1 0.5] Spawn the bullet tracers before updating muzzle location
  (FortWeapon|Effects|SpawnBulletTracerBeamsInternal HitLocations (MuzzleCycling|GetFirePointSocketLocation))
  ;;;[1 1 1 0.5] Update the location of the muzzle flash
  (Transformation|SetWorldLocation _spawned_muzzle_system (Effects|GetMuzzleLocation -1) false false)
  ;;;[1 1 1 0.5] Update the parameter that controls shell ejection
  (Effects|Components|ParticleSystem|SetIntParameter _spawned_muzzle_system "FireIndex" (Variables|MuzzleCycling|GetFireIndex))
  ;;;[1 1 1 0.5] Update FireIndex and beam location
  (CallFunction|OnWeaponFired))

(event Custom|BulletTracerNiagaraDataSent
  (bind _spawned_muzzle_system (Variables|Effects|Muzzle|GetSpawned%20Muzzle%20System))
  ;[1 1 1] Muzzle Cycling Logic
  ;;[0 0 0 0.5] Updated muzzle FX after activating bullet trails.
  ;;;[1 1 1 0.5] Update the location of the muzzle flash
  (Transformation|SetWorldLocation _spawned_muzzle_system (Effects|GetMuzzleLocation -1) false false)
  ;;;[1 1 1 0.5] Update the parameter that controls shell ejection
  (Effects|Components|ParticleSystem|SetIntParameter _spawned_muzzle_system "FireIndex" (Variables|MuzzleCycling|GetFireIndex))
  ;;;[1 1 1 0.5] Update FireIndex and beam location
  (CallFunction|OnWeaponFired))

(event Custom|OnWeaponFired
  (bind _returnvalue (AudioParameter|GetAudioParameterComponent))
  ;[1 1 1] Muzzle Cycling Logic
  ;;[0 0 0 0.5] On Weapon Fired: increment FireIndex
  ;;;[1 1 1 0.5] Reset
  (if (== (Variables|MuzzleCycling|GetFireIndex) (MuzzleCycling|GetMuzzleSocketFinalIndex))
    (MuzzleCycling|UpdateFireIndex 0 "Hit final index resetting index to 0")
    (else
      ;;;[1 1 1 0.5] Otherwise, simply increment FireIndex
      (MuzzleCycling|UpdateFireIndex (+ (Variables|MuzzleCycling|GetFireIndex) 1) "OnWeaponFired")))
  (Utilities|IsValid _returnvalue
    (:Is%20Valid
      (Audio|Parameter|SetIntegerParameter _returnvalue "WeaponIndex" (Variables|MuzzleCycling|GetFireIndex) :!class /Script/AudioExtensions.AudioParameterControllerInterface))))

(event EventSetSpecialShotData
  (params (IsDoingSpecialShot :type Boolean))
  ;[0.8407 0 1] Interface functions to set and reset the special shot data for the
  ; animation blueprints to blend in the shotgun animation
  (bind _output_get (Variables|Default|SetIsDoingSpecialShot IsDoingSpecialShot))
  (if _output_get
    (Utilities|Time|SetTimerbyEvent (EventDispatchers|CreateEvent :!event Custom|ResetSpecialShotData) 0.15 false true)))

(event Custom|ResetSpecialShotData
  ;[0.8407 0 1] Interface functions to set and reset the special shot data for the
  ; animation blueprints to blend in the shotgun animation
  (Variables|Default|SetIsDoingSpecialShot false))

(event EventPlayWeaponFireCustomAudio
  (params (bPersistentFire :type Boolean) (bSecondaryFire :type Boolean))
  (|CallParentFunction bPersistentFire bSecondaryFire)
  (if (Pawn|IsLocallyViewed (Game|GetInstigator))
    (bind _returnvalue (Audio|SpawnSoundAttached "/WildEstateWeapons/Gameplay/ColdFern/Sound/MSS_ColdFern_SurroundSweetner_1P.MSS_ColdFern_SurroundSweetner_1P" (Variables|Visual|GetWeaponMesh)))
    (bind _output_get (Variables|Default|SetWeaponFireSweetenerAudioComponent _returnvalue))
    (Audio|Parameter|SetBooleanParameter _output_get "DoSpecialShot" (Variables|Default|GetIsDoingSpecialShot) :!class /Script/AudioExtensions.AudioParameterControllerInterface)))

(event EventStopWeaponFireCustomAudio
  (|CallParentFunction)
  (if (Pawn|IsLocallyViewed (Game|GetInstigator))
    (bind _weaponfiresweeteneraudiocomponent (Variables|Default|GetWeaponFireSweetenerAudioComponent)
      (:then
        (Audio|Parameter|ExecuteTriggerParameter _weaponfiresweeteneraudiocomponent "OnStop"))
      (:else))))

(fn GetData
  (return (Variables|Default|GetIsDoingSpecialShot)))

(fn GetFirePointSocketLocation
  (bind _fireindex (Variables|MuzzleCycling|GetFireIndex))
  (bind _muzzlesocketnames (Variables|MuzzleCycling|GetMuzzleSocketNames))
  (if (Utilities|Array|IsValidIndex _muzzlesocketnames _fireindex)
    (return (Transformation|GetSocketLocation (Variables|Visual|GetWeaponMesh) (Utilities|Array|Get(aref) _muzzlesocketnames _fireindex)))
    (else
      (Fortnite|FortShippingLog self "[GetFirePointSocketLocation] Invalid FireIndex.  Returning default MuzzleSocketName location." true)
      (return (Transformation|GetSocketLocation (Variables|Visual|GetWeaponMesh) (Effects|GetMuzzleSocketName 0))))))

(fn GetFirePointSocketName
  (params (InFireIndex :type Integer))
  (if (Utilities|Array|IsValidIndex (Variables|MuzzleCycling|GetMuzzleSocketNames) InFireIndex)
    (return (Utilities|Array|Get(aref) (Variables|MuzzleCycling|GetMuzzleSocketNames) InFireIndex))
    (else
      (Fortnite|FortShippingLog self "[GetFirePointSocketName] Invalid FireIndex.  Returning default MuzzleSocketName." true)
      (return (Effects|GetMuzzleSocketName 0)))))

(fn GetMuzzleLocation
  (params (PatternIndex :type Integer))
  (return (MuzzleCycling|GetFirePointSocketLocation)))

(fn GetMuzzleSocketFinalIndex
  (bind _muzzlesocketnames (Variables|MuzzleCycling|GetMuzzleSocketNames))
  (bind _returnvalue (Utilities|Array|Length _muzzlesocketnames))
  (bind _returnvalue_2 (- _returnvalue 1))
  (if (Utilities|Array|IsValidIndex _muzzlesocketnames _returnvalue_2)
    (return _returnvalue_2)
    (else
      (Fortnite|FortShippingLog self "[GetMuzzleSocketFinalIndex] Invalid FireIndex.  Returning 0." true)
      (return 0))))

(fn GetSpecialShotData
  (return (Variables|Default|GetIsDoingSpecialShot)))

(fn UpdateFireIndex
  (params (NewFireIndex :type Integer) (Reason :type String))
  (if (not (Utilities|Array|IsEmpty (Variables|MuzzleCycling|GetMuzzleSocketNames)))
    (bind _output_get (Variables|MuzzleCycling|SetFireIndex (Math|Integer|%25%28Integer%29 NewFireIndex (Utilities|Array|Length (Variables|MuzzleCycling|GetMuzzleSocketNames)))))
    (Effects|SetBeamSourceSocketName (MuzzleCycling|GetFirePointSocketName _output_get))
    (return _output_get)
    (else
      ;[1 0 0 0.5] Error: no muzzle socket names present
      (Fortnite|FortShippingLog self "[UpdateFireIndex] MuzzleSocketNames is empty!  Not updating FireIndex." true)
      (return (Variables|GetNewFireIndex)))))

(fn ConstructionScript
  (|CallParentFunction))
