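; Blueprint for the shotgun half of the ColdFern dual weapon in WildEstate.
; When weapon fire FX play, it runs the parent behavior, then fetches the paired weapon.
; If that paired weapon is valid, it flags it with special shot data through the ColdFern interface message.
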(event EventOnPlayWeaponFireFX
  (params (bPersistentFire :type Boolean) (bSecondaryFire :type Boolean))
  (bind _returnvalue (CallFunction|GetPairedWeapon))
  (|CallParentFunction bPersistentFire bSecondaryFire)
  (Utilities|IsValid _returnvalue
    (:Is%20Valid
      (Class|BPColdFernInterface|SetSpecialShotData(Message) _returnvalue true))))

(fn ConstructionScript
  (|CallParentFunction))
